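; Lifesteal gameplay ability for Donut guns that restores the owner's health, and optionally shield, from damage dealt.
; Only events with Donut instigator tags and non-building, non-vehicle targets trigger it.
; Event magnitude times a data-row multiplier fills missing health, with overflow going to shield if enabled, both floored.
; One SetByCaller effect applies both amounts, after a heal cue fires with pre-heal health ratio for correct full-HP FX.
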(event Ability|EventActivateAbilityFromEvent
  (params (EventData :type Gameplay%20Event%20Data%20Structure%20%28by%20ref%29))
  (bind (_eventtag _instigator _target _optionalobject _optionalobject2 _contexthandle _instigatortags _targettags _eventmagnitude) (Utilities|Struct|BreakGameplayEventData EventData))
  (bind _player_pawn (Variables|Default|GetPlayer_Pawn))
  (bind _returnvalue (FortPawn|GetHealth _player_pawn))
  (bind _returnvalue_2 (FortPawn|GetMaxHealth _player_pawn))
  (bind _returnvalue_3 (- _returnvalue_2 _returnvalue))
  (bind _row_heal_mult (Variables|Default|GetRow_Heal_Mult))
  (bind _returnvalue_4 (ScalableFloat|GetValueatLevel _row_heal_mult))
  (bind _returnvalue_5 (* _eventmagnitude _returnvalue_4))
  (bind _returnvalue_6 (- _returnvalue_5 _returnvalue_3))
  (bind _returnvalue_7 (Math|Float|Clamp(Float) _returnvalue_6 0.0 1e+04))
  (bind _returnvalue_8 (Math|Float|Clamp(Float) _returnvalue_5 0.0 _returnvalue_3))
  (bind _returnvalue_9 (FortPawn|GetShield _player_pawn))
  (bind _returnvalue_10 (FortPawn|GetMaxShield _player_pawn))
  (bind _returnvalue_11 (- _returnvalue_10 _returnvalue_9))
  (bind _returnvalue_12 (Math|Float|Clamp(Float) _returnvalue_7 0.0 _returnvalue_11))
  (bind _returnvalue_13 (Math|Float|Floor _returnvalue_8))
  (bind _returnvalue_14 (Math|Conversions|ToFloat(Integer) _returnvalue_13))
  (bind _returnvalue_15 (Math|Float|Floor _returnvalue_12))
  (bind _returnvalue_16 (Math|Conversions|ToFloat(Integer) _returnvalue_15))
  (bind _row_shouldhealshields (Variables|Default|GetRow_ShouldHealShields))
  (bind _returnvalue_17 (Class|AthenaFunctionLibrary|RowtoBool _row_shouldhealshields))
  (bind _returnvalue_18 (select _returnvalue_17 _returnvalue_16 0.0))
  ;[1 1 1] Save out player pawn if we have not already
  (Utilities|IsValid (Variables|Default|GetPlayer_Pawn)
    (:Is%20Valid
      _)
    (:Is%20Not%20Valid
      (bind _asfort_player_pawn (Utilities|Casting|CastToFortPlayerPawn (Ability|GetAvatarActorfromActorInfo)))
      (Variables|Default|SetPlayer_Pawn _asfort_player_pawn)))
  (Development|PrintString "Donut Dmg Hit")
  ;[1 1 1] Create effect spec and assign mag for healing and shielding
  (bind _returnvalue_19 (Ability|GameplayEffect|AssignTagSetbyCallerMagnitude (Ability|MakeOutgoingGameplayEffectSpec (Variables|Default|GetGE_Heal)) "(TagName=\"SetByCaller.Healing\")" _returnvalue_14))
  (bind _returnvalue_20 (Ability|GameplayEffect|AssignTagSetbyCallerMagnitude _returnvalue_19 "(TagName=\"SetByCaller.Shield\")" _returnvalue_18))
  ;[1 1 1] If we need to heal health or shield apply effect and play gameplaycue. The Cue needs to fire first so that the correct FX play on full HP
  (if (or (> _returnvalue_14 0.0) (> _returnvalue_18 0.0))
    (Ability|ExecuteGameplayCueWithParamsOnOwner (Variables|Default|GetGC_HealMinor) (Ability|GameplayCue|MakeGameplayCueParameters (/ (FortPawn|GetHealth (Variables|Default|GetPlayer_Pawn)) (FortPawn|GetMaxHealth (Variables|Default|GetPlayer_Pawn)))))
    (Ability|ApplyGameplayEffectSpecToOwner _returnvalue_20)))

(fn ShouldAbilityRespondToEvent
  (params (ActorInfo :type Gameplay%20Ability%20Actor%20Info%20Structure) (Payload :type Gameplay%20Event%20Data%20Structure))
  (bind (_eventtag _instigator _target _optionalobject _optionalobject2 _contexthandle _instigatortags) (Utilities|Struct|BreakGameplayEventData Payload))
  ;[1 1 1] Check if we did damage with the Donut guns if so check if we hit a building actor and if we hit a vehicle. If we did not hit either then trigger the ability.
  (if (GameplayTags|TagContainer|CompareTagContainertoOtherTags (Variables|Default|GetT_Donut) _instigatortags)
    (Utilities|Casting|CastToBuildingActor _target
      (:then
        _)
      (:CastFailed
        (bind _returnvalue (GameplayTags|HasMatchingGameplayTag(Message) _target (Variables|Default|GetT_Vehicle)))
        (return (not _returnvalue))))
    (else
      _))
  (return false))
